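# Anim Blueprint for the Bistro Monster backpack cosmetic, animating its equip, idle and transform-back-to-astronaut sequence.
# The state machine enters at Equip, auto-advances to a weighted-random Idle, then moves to Transform To Astronaut when the Transform bool is set.
# The event graph sets Transform on the "aie.BistroBackpackOutro" anim input event, and DefaultSlot sits after the state machine for montages.
# The testLoop states form a separate chain unreachable from the entry.
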AnimBlueprint /BRCosmetics/Accessories/FORT_Backpacks/Backpack_F_MED_Bistro_Monster/Meshes/F_MED_Bistro_Monster_Pack_AnimBP
target_skeleton=&/F_MED_Bistro_Monster_Pack_Skeleton
graphs [
 AnimGraph {nodes [
   Input Pose()
   AnimGraphNode_StateMachine_0 = New State Machine() {is=title AnimGraphNode_StateMachine}
   Output Pose(Result=Slot 'DefaultSlot'(Source=@AnimGraphNode_StateMachine_0)) {layerGroup=None}
  ]}
 New State Machine {
  entry=Equip
  path=AnimGraph.AnimGraphNode_StateMachine_0.New State Machine
  ids=id
  nodes [
   AnimStateNode_0 = testLoop_01() {is=title AnimStateNode}
   AnimStateNode_1 = testLoop_02() {is=title AnimStateNode}
   AnimStateNode_2 = Idle()
   AnimStateNode_3 = Equip() {is=title AnimStateNode}
   AnimStateNode_4 = testLoop_04() {is=title AnimStateNode}
   AnimStateNode_5 = testLoop_03() {is=title AnimStateNode}
   AnimStateNode_6 = Transform To Astronaut() {is=title AnimStateNode}
   AnimStateTransitionNode_0 = AnimStateTransitionNode() {bAutomaticRuleBasedOnSequencePlayerInState=true, crossfadeDuration=0}
   AnimStateTransitionNode_1 = AnimStateTransitionNode() {bAutomaticRuleBasedOnSequencePlayerInState=true}
   AnimStateTransitionNode_2 = AnimStateTransitionNode() {bAutomaticRuleBasedOnSequencePlayerInState=true, crossfadeDuration=0}
   AnimStateTransitionNode_3 = AnimStateTransitionNode() {bAutomaticRuleBasedOnSequencePlayerInState=true, crossfadeDuration=0}
   AnimStateTransitionNode_4 = AnimStateTransitionNode() {bAutomaticRuleBasedOnSequencePlayerInState=true, crossfadeDuration=0}
   AnimStateTransitionNode_5 = AnimStateTransitionNode()
  ]
  links [
   {
    dst=Idle
    rule=AnimGraph.AnimGraphNode_StateMachine_0.New State Machine.AnimStateTransitionNode_1.AnimationTransitionGraph_0
    src=Equip
    via=AnimStateTransitionNode_1
   }
   {
    dst=Transform To Astronaut
    rule=AnimGraph.AnimGraphNode_StateMachine_0.New State Machine.AnimStateTransitionNode_5.Transition
    src=Idle
    via=AnimStateTransitionNode_5
   }
   {
    dst=testLoop_02
    rule=AnimGraph.AnimGraphNode_StateMachine_0.New State Machine.AnimStateTransitionNode_2.Transition
    src=testLoop_01
    via=AnimStateTransitionNode_2
   }
   {
    dst=testLoop_03
    rule=AnimGraph.AnimGraphNode_StateMachine_0.New State Machine.AnimStateTransitionNode_0.AnimationTransitionGraph_0
    src=testLoop_02
    via=AnimStateTransitionNode_0
   }
   {
    dst=testLoop_04
    rule=AnimGraph.AnimGraphNode_StateMachine_0.New State Machine.AnimStateTransitionNode_3.Transition
    src=testLoop_03
    via=AnimStateTransitionNode_3
   }
   {
    dst=testLoop_02
    rule=AnimGraph.AnimGraphNode_StateMachine_0.New State Machine.AnimStateTransitionNode_4.Transition
    src=testLoop_04
    via=AnimStateTransitionNode_4
   }
  ]
 }
 testLoop_01 {
  path=AnimGraph.AnimGraphNode_StateMachine_0.New State Machine.AnimStateNode_0.testLoop_01
  nodes [
   n1 = Sequence Player 'Bistro_Monster_idle_BACKPACK'() {bLoopAnimation=false, sequence=/BRCosmetics/Accessories/FORT_Backpacks/Backpack_F_MED_Bistro_Monster/Animation/Bistro_Monster_idle_BACKPACK}
   Output Animation Pose(Result=@n1)
  ]
 }
 testLoop_02 {
  path=AnimGraph.AnimGraphNode_StateMachine_0.New State Machine.AnimStateNode_1.testLoop_02
  nodes [
   n1 = Sequence Player 'Emote_BistroStyle_Switch_Backpack'() {
    bLoopAnimation=false
    sequence=/BRCosmetics/Accessories/FORT_Backpacks/Backpack_F_MED_Bistro_Monster/Animation/Emote_BistroStyle_Switch_Backpack
   }
   Output Animation Pose(Result=@n1)
  ]
 }
 Idle {
  path=AnimGraph.AnimGraphNode_StateMachine_0.New State Machine.AnimStateNode_2.Idle
  nodes [
   n1 = Random Sequence Player!() -> (Pose:Pose Link Structure) {
    is=title AnimGraphNode_RandomPlayer
    entries [
     {
      chanceToPlay=0.8
      maxLoopCount=0
      maxPlayRate=1
      minLoopCount=0
      minPlayRate=1
      sequence=&Bistro_Monster_idle_BACKPACK
      blendIn {blendOption=Linear, blendTime=0.2, customCurve=null}
     }
     {
      chanceToPlay=0.2
      maxLoopCount=0
      maxPlayRate=1
      minLoopCount=0
      minPlayRate=1
      sequence=/BRCosmetics/Accessories/FORT_Backpacks/Backpack_F_MED_Bistro_Monster/Animation/Bistro_Monster_idle_break_BACKPACK
      blendIn {blendOption=Linear, blendTime=0.2, customCurve=null}
     }
    ]
   }
   Output Animation Pose(Result=@n1)
  ]
 }
 Equip {
  path=AnimGraph.AnimGraphNode_StateMachine_0.New State Machine.AnimStateNode_3.Equip
  nodes [Output Animation Pose(Result=Sequence Player 'Emote_BistroStyle_Switch_Monster_Backpack'() {sequence=/BRCosmetics/Accessories/FORT_Backpacks/Backpack_F_MED_Bistro_Monster/Animation/Emote_BistroStyle_Switch_Monster_Backpack})]
 }
 testLoop_04 {
  path=AnimGraph.AnimGraphNode_StateMachine_0.New State Machine.AnimStateNode_4.testLoop_04
  nodes [
   n1 = Sequence Player 'Bistro_Monster_idle_break_BACKPACK'() {
    bLoopAnimation=false
    sequence=&Bistro_Monster_idle_break_BACKPACK
   }
   Output Animation Pose(Result=@n1)
  ]
 }
 testLoop_03 {
  path=AnimGraph.AnimGraphNode_StateMachine_0.New State Machine.AnimStateNode_5.testLoop_03
  nodes [
   n1 = Sequence Player 'Bistro_Monster_idle_BACKPACK'() {bLoopAnimation=false, sequence=&Bistro_Monster_idle_BACKPACK}
   Output Animation Pose(Result=@n1)
  ]
 }
 Transform To Astronaut {
  path=AnimGraph.AnimGraphNode_StateMachine_0.New State Machine.AnimStateNode_6.Transform To Astronaut
  nodes [
   n1 = Sequence Player 'Emote_BistroStyle_Switch_Monster_Return_Backpack'() {
    bLoopAnimation=false
    sequence=/BRCosmetics/Accessories/FORT_Backpacks/Backpack_F_MED_Bistro_Monster/Animation/Emote_BistroStyle_Switch_Monster_Return_Backpack
   }
   Output Animation Pose(Result=@n1)
  ]
 }
 AnimationTransitionGraph_0 {path=AnimGraph.AnimGraphNode_StateMachine_0.New State Machine.AnimStateTransitionNode_0.AnimationTransitionGraph_0, nodes [Result()]}
 AnimationTransitionGraph_0 {path=AnimGraph.AnimGraphNode_StateMachine_0.New State Machine.AnimStateTransitionNode_1.AnimationTransitionGraph_0, nodes [Result()]}
 Transition {path=AnimGraph.AnimGraphNode_StateMachine_0.New State Machine.AnimStateTransitionNode_2.Transition, nodes [Result()]}
 Transition {path=AnimGraph.AnimGraphNode_StateMachine_0.New State Machine.AnimStateTransitionNode_3.Transition, nodes [Result()]}
 Transition {path=AnimGraph.AnimGraphNode_StateMachine_0.New State Machine.AnimStateTransitionNode_4.Transition, nodes [Result()]}
 Transition {
  path=AnimGraph.AnimGraphNode_StateMachine_0.New State Machine.AnimStateTransitionNode_5.Transition
  nodes [Result(bCanEnterTransition=Get Transform() -> (Transform:bool))]
 }
 EventGraph {nodes [
   n1 = Event BlueprintProcessAnimInputEvent()
   > Branch(Condition=IsNamed(self=@n1.AnimInputEvent, InEventName=aie.BistroBackpackOutro))
   > Set Transform(Transform:bool=true) -> (Output_Get:bool)
   TryGetPawnOwner()
   Event BlueprintUpdateAnimation()
  ]}
]
variables [Transform {type=boolean}]
